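# Niagara ribbon trail effect for the MediCrow pickaxe cosmetic.
# One anim-trail emitter spawns short-lived particles from bone or socket pairs and draws them with three ribbon renderers.
# User.Variant blends between two color variants and tints a separate under-trail color from dark purple to green, faded over particle age.
# Time-of-day scalars dim alpha plus red and green at night and evening, leaving blue unscaled.
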NiagaraSystem /BRCosmetics/Weapons/FORT_Melee/Pickaxe_MediCrow/FX/NS_Pickaxe_MediCrow_Trail
dependencies {
 usedDataInterfaces [&NiagaraDataInterfaceColorCurve, &NiagaraDataInterfaceCurve, &NiagaraDataInterfaceSocketReader]
 usedDynamicInputs []
 usedModules [
  &self:SystemSpawnScript.SystemScriptSource.SystemScriptGraph.NiagaraNodeAssignment_1.SetVariables_#1
  &SystemState
 ]
}
emitters {FN_AnimTrail {
  emitterSpawnScript [&FN_RibbonIDOffsetUtility()]
  emitterUpdateScript [
   &EmitterState(Life Cycle Mode=::NewEnumerator1, Loop Behavior=::NewEnumerator1, Loop Duration=1, UseLoopDelay=false)
   &FN_SpawnAnimTrail(Number of Anim Trails=2)
  ]
  particleSpawnScript [
   &FN_InitializeAnimTrailBoneIndecies()
   &FN_InitializeAnimTrailBoneIndecies(AnimTrail Selection=1, First Bone or Socket Index=2, Second Bone or Socket Index=3) {enabled=false, module=FN_InitializeAnimTrailBoneIndecies001}
   &FN_InitializeAnimTrail(Lifetime=0.22)
   &AddVelocity(Velocity=<0 400 0>) {enabled=false}
  ]
  particleUpdateScript [
   &SolveForcesAndVelocity(Clamp Velocity=false, Limit Acceleration=false)
   /SetVariables(
    Particles.Variant0_CC:LinearColor=&ColorFromCurve(ColorCurve=DataInterface() {propertyValues {
       AlphaCurve {keys []}
       BlueCurve {keys [<at 0.1524 0.2487 RCIM_Cubic>]}
       GreenCurve {keys [<at 0.1524 0.2487 RCIM_Cubic>]}
       RedCurve {keys [<at 0.1524 0.2487 RCIM_Cubic>]}
      }})
    Particles.Variant1_CC:LinearColor=&ColorFromCurve(ColorCurve=DataInterface() {propertyValues {
       AlphaCurve {keys []}
       BlueCurve {keys [<at 0.1524 0.2487 RCIM_Cubic>]}
       GreenCurve {keys [<at 0.1524 0.2487 RCIM_Cubic>]}
       RedCurve {keys [<at 0.1524 0.2487 RCIM_Cubic>]}
      }})
   ) {isSetParameters=true}
   &FN_UpdateAnimTrail() {module=FN_UpdateAnimTrail001}
   /SetVariables(Particles.UnderTrail:LinearColor=&ScaleLinearColorByCurve(
     InitialValue=&LerpLinearColors(StartColor=<rgba 0.0203 0.0063 0.0187 1>, EndColor=<rgba 0.0842 0.3383 0.0052 1>, LerpFactor=$User.Variant)
     ScaleCurve=DataInterface() {propertyValues {
       AlphaCurve {keys [<at 0 1>,<at 0.9938 0>]}
       BlueCurve {keys [<at 0.2224 1>]}
       GreenCurve {keys [<at 0.2224 1>]}
       RedCurve {keys [<at 0.2224 1>]}
      }}
     CurveIndex=$Particles.NormalizedAge
    )) {isSetParameters=true}
   &ScaleColor(
    Scale Alpha=&Multiply_Float(
     A=&FloatFromCurve(FloatCurve=DataInterface() {propertyValues {Curve {keys [
          {arriveTangent=9.0725, interpMode=RCIM_Cubic, leaveTangent=9.0725, tangentMode=RCTM_User, time=0, value=0}
          {arriveTangent=-2.5249, interpMode=RCIM_Cubic, leaveTangent=-2.5249, tangentMode=RCTM_User, time=0.1408, value=0.9847}
          {arriveTangent=-0.0405, interpMode=RCIM_Cubic, leaveTangent=-0.0405, tangentMode=RCTM_User, time=1, value=0}
         ]}}})
     B=&FNTimeOfDayScalar(Night Scale=0.5, Morning Scale=1, Evening Scale=0.5)
    )
    Scale RGB=&MakeVectorFromLinearColorRGB(Color=&MultiplyLinearColor(
      Linear Color=&LerpLinearColors(StartColor=$Particles.Variant0_CC, EndColor=$Particles.Variant1_CC, LerpFactor=$User.Variant)
      Scale Factor=&MakeVector4(
       X=&FNTimeOfDayScalar(Night Scale=0.5, Morning Scale=1, Evening Scale=0.5)
       Y=&FNTimeOfDayScalar(Night Scale=0.5, Morning Scale=1, Evening Scale=0.5)
       Z=1
       W=1
      )
     ))
    ScaleA=true
    ScaleRGB=true
   ) {module=ScaleColor002}
  ]
  rendererClasses [&NiagaraRibbonRendererProperties]
  renderers [
   {rendererClass=&NiagaraRibbonRendererProperties}
   {rendererClass=&NiagaraRibbonRendererProperties, rendererIndex=1}
   {rendererClass=&NiagaraRibbonRendererProperties, rendererIndex=2}
  ]
 }}
summary {
 systemName=NS_Pickaxe_MediCrow_Trail
 emitters [{emitterName=FN_AnimTrail, rendererClasses [&NiagaraRibbonRendererProperties]}]
 userVariables [
  User.bIsFrontEnd {type=NiagaraBool}
  User.bIsFrontEndPreview {type=NiagaraBool}
  User.RibbonIDOffset {type=NiagaraInt32}
  User.Variant {type=NiagaraFloat, defaultValue=0, description=""}
 ]
}
